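; Spawner Blueprint for the normal Group 0 chest in the VividRazor vault, scaling its spawn chance level with team size.
; On server BeginPlay it waits for playlist data, clamps max team size to 1–4, sets the level to one less, then runs parent BeginPlay.
; The spawn override casts the chosen class to the chest type, spawns it with the vault location tag and reports it through ServerActorSpawned.
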(event EventBeginPlay
  ; Begin Play
  (if (Networking|IsServer)
    (bind (_gamestate _playlist) (Class|FortAsyncActionCurrentPlaylistReady|PlaylistDataReadyAsync)
      (:OnReady
        (bind _output_get (Variables|Default|SetMax%20Team%20Size (Math|Integer|Clamp(Integer) (Variables|/Script/FortniteGame.FortPlaylist|GetMaxTeamSize _playlist) 1 4)))
        (Variables|SpawnerVars|SetSpawnChanceLevel (Math|Conversions|ToFloat(Integer) (- _output_get 1)))
        (|CallParentFunction)))))

(event Collision|EventActorBeginOverlap
  (params (OtherActor :type Actor%20Object%20Reference)))

(event EventTick
  (params (DeltaSeconds :type Float%20%28single-precision%29)))

(event EventSpawnActor
  (params (New%20Transform :type Transform))
  ; Spawn Actor: Override
  (bind _returnvalue (CallFunction|GetActorClasstoSpawn))
  (if (Utilities|IsValidClass _returnvalue)
    (bind _asbga_vivid_razor_vault_chest_normal (Utilities|Casting|CastToBGA_VividRazorVault_Chest_NormalClass _returnvalue))
    (bind _returnvalue_2 (Game|SpawnBuildingActorfromClass _asbga_vivid_razor_vault_chest_normal New%20Transform 0 false (Variables|Default|GetVaultLocationTag) 0 0 0))
    (CallFunction|ServerActorSpawned _returnvalue_2)))

(fn ConstructionScript
  (|CallParentFunction))
